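; Editor tool that bakes a shadow proxy atlas for static meshes of listed actor classes, which can be auto-filled from tree props in the level.
; Each unique mesh is scene-captured from four tilt angles, composited 2x2, then material-converted into a distance-field RGBA atlas frame.
; Actor locations, with bounds radius and atlas frame packed into transform scale, feed BP_DisplayShadowProxy_01's instances.
; A separate event saves the atlas as a static texture.
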(event Custom|Generate
  (bind _framesxy (Variables|Default|GetFramesXY))
  (bind _returnvalue (Math|Conversions|ToFloat(Integer) _framesxy))
  (bind _framesxy_2 (Variables|Default|GetFramesXY))
  (bind _mesharray (Variables|Default|GetMeshArray))
  (Utilities|Set|Clear (Variables|Default|GetMeshes))
  (Variables|Default|SetVertOffset 0)
  (Utilities|Array|Clear (Variables|Default|GetVerts))
  (Utilities|Array|Clear (Variables|Default|GetTris))
  (Utilities|Array|Clear (Variables|Default|GetUVs))
  (Utilities|Array|Clear (Variables|Default|GetUVs1))
  (Utilities|Array|Clear (Variables|Default|GetUVs2))
  (Utilities|Array|Clear (Variables|Default|GetActors))
  (Utilities|Array|Clear (Variables|Default|GetGatherInstanceData))
  (for _array_element (Variables|Default|GetClasses)
    (bind _outactors (Actor|GetAllActorsOfClass _array_element))
    (for _array_element_2 _outactors
      (Utilities|Array|AddUnique (Variables|Default|GetActors) _array_element_2)
      (Utilities|Set|Add (Variables|Default|GetMeshes) (Variables|/Script/Engine.StaticMeshComponent|GetStaticMesh (Actor|GetComponentbyClass _array_element_2 "/Script/Engine.StaticMeshComponent")))))
  (bind _result (Utilities|Set|ToArray (Variables|Default|GetMeshes)))
  (Variables|Default|SetMeshArray _result)
  (CallFunction|CreateRTsifNull)
  (CallFunction|CaptureShadows)
  (for _array_element_3 (Variables|Default|GetActors)
    (Utilities|Array|Add (Variables|Default|GetGatherInstanceData) (Math|Transform|MakeTransform (.location (Transformation|GetActorTransform _array_element_3)) "0, 0, 0" (Math|Vector|MakeVector (.SphereRadius (Collision|GetComponentBounds (Actor|GetComponentbyClass _array_element_3 "/Script/Engine.StaticMeshComponent"))) (Math|Vector2D|BreakVector2D (+ (/ (Utilities|1dTo2dIndex (Utilities|Array|FindItem _mesharray (Variables|/Script/Engine.StaticMeshComponent|GetStaticMesh (Actor|GetComponentbyClass _array_element_3 "/Script/Engine.StaticMeshComponent"))) _framesxy_2) _returnvalue) "(X=0.001000,Y=0.001000)")) (.Y (Math|Vector2D|BreakVector2D (+ (/ (Utilities|1dTo2dIndex (Utilities|Array|FindItem _mesharray (Variables|/Script/Engine.StaticMeshComponent|GetStaticMesh (Actor|GetComponentbyClass _array_element_3 "/Script/Engine.StaticMeshComponent"))) _framesxy_2) _returnvalue) "(X=0.001000,Y=0.001000)"))))))
    (Development|PrintString (Utilities|String|Append (Utilities|String|Append (Utilities|GetDisplayName (Variables|/Script/Engine.StaticMeshComponent|GetStaticMesh (Actor|GetComponentbyClass _array_element_3 "/Script/Engine.StaticMeshComponent"))) (Utilities|String|Append "XYZ=" (Utilities|String|ToString(Transform) (Math|Transform|MakeTransform (.location (Transformation|GetActorTransform _array_element_3)) "0, 0, 0" (Math|Vector|MakeVector (.SphereRadius (Collision|GetComponentBounds (Actor|GetComponentbyClass _array_element_3 "/Script/Engine.StaticMeshComponent"))) (Math|Vector2D|BreakVector2D (+ (/ (Utilities|1dTo2dIndex (Utilities|Array|FindItem _mesharray (Variables|/Script/Engine.StaticMeshComponent|GetStaticMesh (Actor|GetComponentbyClass _array_element_3 "/Script/Engine.StaticMeshComponent"))) _framesxy_2) _returnvalue) "(X=0.001000,Y=0.001000)")) (.Y (Math|Vector2D|BreakVector2D (+ (/ (Utilities|1dTo2dIndex (Utilities|Array|FindItem _mesharray (Variables|/Script/Engine.StaticMeshComponent|GetStaticMesh (Actor|GetComponentbyClass _array_element_3 "/Script/Engine.StaticMeshComponent"))) _framesxy_2) _returnvalue) "(X=0.001000,Y=0.001000)")))))))) (Utilities|String|ToString(Vector2d) (/ (Utilities|1dTo2dIndex (Utilities|Array|FindItem _mesharray (Variables|/Script/Engine.StaticMeshComponent|GetStaticMesh (Actor|GetComponentbyClass _array_element_3 "/Script/Engine.StaticMeshComponent"))) _framesxy_2) _returnvalue)))))
  (bind _outactors_2 (Actor|GetAllActorsOfClass "/Game/ContentCreationTools/ShadowProxyGenerator/Blueprints/BP_DisplayShadowProxy_01.BP_DisplayShadowProxy_01_C"))
  (Variables|/Game/ContentCreationTools/ShadowProxyGenerator/Blueprints/BP_DisplayShadowProxy_01.BP_DisplayShadowProxy_01_C|SetInstanceData (Utilities|Struct|MakeFakeShadowData (Variables|Default|GetGatherInstanceData)) (Utilities|Array|Get(acopy) _outactors_2))
  (Class|BPDisplayShadowProxy01|UpdateInstances (Utilities|Array|Get(acopy) _outactors_2))
  (Variables|Default|SetRequiredSize (Math|Float|Ceil (Math|Float|Sqrt (Math|Conversions|ToFloat(Integer) (Utilities|Set|Length (Variables|Default|GetMeshes)))))))

(event Custom|Create%20Texture
  (bind _returnvalue (EditorScripting|Rendering|RenderTarget2DCreateStaticTexture2DEditorOnly (Variables|Default|GetFinalAtlas%20RT) (Utilities|String|Append "/Game/ContentCreationTools/ShadowProxyGenerator/Textures/" (Variables|Default|GetTexture%20Name))))
  (Development|PrintString (Utilities|GetObjectPathString _returnvalue)))

(event Custom|CompositeSingleFrame
  (bind _returnvalue (Math|Conversions|ToFloat(Integer) 2))
  (bind _angleidx (Variables|Default|Getangleidx))
  (bind _2d_index (Utilities|1dTo2dIndex _angleidx 2))
  (bind (_canvas _size _context) (Rendering|BeginDrawCanvastoRenderTarget (Variables|Default|GetViewComposite%20RT)))
  (Canvas|DrawMaterial :self _canvas :RenderMaterial (Variables|Default|GetSampleCaptureMID) :ScreenPosition (* (Math|Vector2D|MakeVector2D (/ (Math|Vector2D|BreakVector2D _size) _returnvalue) (/ (.Y (Math|Vector2D|BreakVector2D _size)) _returnvalue)) _2d_index) :ScreenSize (Math|Vector2D|MakeVector2D (/ (Math|Vector2D|BreakVector2D _size) _returnvalue) (/ (.Y (Math|Vector2D|BreakVector2D _size)) _returnvalue)) :CoordinateSize "(X=1.000000,Y=1.000000)" :Rotation 0.0 :PivotPoint "(X=0.500000,Y=0.500000)")
  (Rendering|EndDrawCanvastoRenderTarget _context)
  (Development|PrintString (Utilities|String|ToString(Vector2d) (* (Math|Vector2D|MakeVector2D (/ (Math|Vector2D|BreakVector2D _size) _returnvalue) (/ (.Y (Math|Vector2D|BreakVector2D _size)) _returnvalue)) _2d_index))))

(event Custom|CopyFrameToAtlas
  (bind (_canvas _size _context) (Rendering|BeginDrawCanvastoRenderTarget (Variables|Default|GetFinalAtlas%20RT)))
  (Canvas|DrawMaterial :self _canvas :RenderMaterial (Variables|Default|GetConvertTo_DistanceField_MID) :ScreenPosition (* (Math|Vector2D|MakeVector2D (/ (Math|Vector2D|BreakVector2D _size) (Math|Conversions|ToFloat(Integer) (Variables|Default|GetFramesXY))) (/ (.Y (Math|Vector2D|BreakVector2D _size)) (Math|Conversions|ToFloat(Integer) (Variables|Default|GetFramesXY)))) (Utilities|1dTo2dIndex (Variables|Default|Getmeshidx) (Variables|Default|GetFramesXY))) :ScreenSize (Math|Vector2D|MakeVector2D (/ (Math|Vector2D|BreakVector2D _size) (Math|Conversions|ToFloat(Integer) (Variables|Default|GetFramesXY))) (/ (.Y (Math|Vector2D|BreakVector2D _size)) (Math|Conversions|ToFloat(Integer) (Variables|Default|GetFramesXY)))) :CoordinateSize "(X=1.000000,Y=1.000000)" :Rotation 0.0 :PivotPoint "(X=0.500000,Y=0.500000)")
  (Rendering|EndDrawCanvastoRenderTarget _context))

(event Custom|GrabAllTreesFromLevel
  (Utilities|FlowControl|Sequence
    (:then_0
      (Utilities|Array|Clear (Variables|Default|GetClasses))
      (bind _outactors (Actor|GetAllActorsOfClass "/Script/FortniteGame.BuildingProp"))
      (for _array_element _outactors
        (if (Utilities|String|Contains (Utilities|GetDisplayName _array_element) "_Tree_" false true)
          (Utilities|Array|AddUnique (Variables|Default|GetClasses) (Utilities|GetClass _array_element)))))
    (:then_1
      (Variables|Default|SetClasses (Variables|Default|GetClasses)))))

(fn CaptureShadows
  (bind _scenecapturecomponent2d (Variables|Default|GetSceneCaptureComponent2D))
  (Rendering|ClearRenderTarget2D (Variables|Default|GetFinalAtlas%20RT))
  (Rendering|ClearRenderTarget2D (Variables|Default|GetViewComposite%20RT))
  (bind _result (Utilities|Set|ToArray (Variables|Default|GetMeshes)))
  (for (_array_element _array_index) _result
    (Variables|Default|Setmeshidx _array_index)
    (Components|StaticMesh|SetStaticMesh (Variables|Default|GetStaticMesh) _array_element)
    (Variables|/Script/Engine.SceneCaptureComponent2D|SetFOVAngle (* (Math|Trig|Atan(Degrees) (/ (.SphereRadius (Math|BoxSphereBounds|BreakBoxSphereBounds (StaticMesh|GetBounds _array_element))) (* (* (.SphereRadius (Math|BoxSphereBounds|BreakBoxSphereBounds (StaticMesh|GetBounds _array_element))) 2.0) 4.0))) 2) _scenecapturecomponent2d)
    (Variables|/Script/Engine.SceneCaptureComponent2D|SetOrthoWidth (* (.SphereRadius (Math|BoxSphereBounds|BreakBoxSphereBounds (StaticMesh|GetBounds _array_element))) 2.0) _scenecapturecomponent2d)
    (Rendering|ClearRenderTarget2D (Variables|Default|GetViewComposite%20RT))
    (for (_array_element_2 _array_index_2) (Variables|Debug|GetCaptureVectors)
      (Variables|Default|Setangleidx _array_index_2)
      (Transformation|SetWorldTransform (Variables|Default|GetSceneCaptureComponent2D) (Math|Transform|MakeTransform (+ (Math|BoxSphereBounds|BreakBoxSphereBounds (StaticMesh|GetBounds _array_element)) (* _array_element_2 (* (* (.SphereRadius (Math|BoxSphereBounds|BreakBoxSphereBounds (StaticMesh|GetBounds _array_element))) 2.0) 4.0))) (Math|Conversions|RotationFromXVector (* _array_element_2 -1.0))) false false)
      (Rendering|ClearRenderTarget2D (Variables|Default|GetSceneCapture%20RT))
      (Rendering|SceneCapture|CaptureScene (Variables|Default|GetSceneCaptureComponent2D) :!class /Script/Engine.SceneCaptureComponent2D)
      (CallFunction|CompositeSingleFrame))
    (CallFunction|CopyFrametoAtlas))
  (for _index (range (Utilities|Array|Length _result) (+ (* (Variables|Default|GetFramesXY) (Variables|Default|GetFramesXY)) 1))
    (Variables|Default|Setmeshidx _index)
    (Rendering|ClearRenderTarget2D (Variables|Default|GetViewComposite%20RT) "(R=0.000000,G=0.000000,B=0.000000,A=0.000000)")
    (CallFunction|CopyFrametoAtlas)))

(fn CreateRTsIfNull
  (bind _finalatlas_rt (Variables|Default|GetFinalAtlas%20RT))
  (Utilities|FlowControl|Sequence
    (:then_0
      (Utilities|IsValid (Variables|Default|GetSceneCapture%20RT)
        (:Is%20Valid
          _)
        (:Is%20Not%20Valid
          (bind _returnvalue (Rendering|CreateRenderTarget2D 256))
          (bind _output_get (Variables|Default|SetSceneCapture%20RT _returnvalue))))
      (Variables|/Script/Engine.SceneCaptureComponent2D|SetTextureTarget _output_get (Variables|Default|GetSceneCaptureComponent2D))
      (Rendering|Material|SetTextureParameterValue (Variables|Default|GetSampleCaptureMID) "RT" (Variables|Default|GetSceneCapture%20RT)))
    (:then_1
      (Utilities|IsValid (Variables|Default|GetViewComposite%20RT)
        (:Is%20Valid
          _)
        (:Is%20Not%20Valid
          (bind _returnvalue_2 (Rendering|CreateRenderTarget2D 512 512))
          (Variables|Default|SetViewComposite%20RT _returnvalue_2)))
      (Rendering|Material|SetTextureParameterValue (Variables|Default|GetConvertTo_DistanceField_MID) "RT" (Variables|Default|GetViewComposite%20RT)))
    (:then_2
      (Utilities|IsValid _finalatlas_rt
        (:Is%20Valid
          (if (== (Variables|/Script/Engine.TextureRenderTarget2D|GetSizeX _finalatlas_rt) 1024)
            _
            (else
              _)))
        (:Is%20Not%20Valid
          _))
      (bind _returnvalue_3 (Rendering|CreateRenderTarget2D 1024 1024 "RTF_RGBA16f" "(R=0.000000,G=0.000000,B=0.000000,A=0.000000)"))
      (Variables|Default|SetFinalAtlas%20RT _returnvalue_3))))

(fn ConstructionScript
  (bind _previewblendmid (Variables|Default|GetPreviewBlendMID))
  (Rendering|SceneCapture|ShowOnlyComponent (Variables|Default|GetSceneCaptureComponent2D) (Variables|Default|GetStaticMesh))
  (bind _returnvalue (Rendering|Material|CreateDynamicMaterialInstance "/Game/ContentCreationTools/ShadowProxyGenerator/Materials/M_Sample_Frame_DFAlpha.M_Sample_Frame_DFAlpha" "None" "None" :!class /Script/Engine.KismetMaterialLibrary))
  (Variables|Default|SetSampleCaptureMID _returnvalue) ; Samples a frame from scene capture, converting to Distance Field Alpha
  (bind _returnvalue_2 (Rendering|Material|CreateDynamicMaterialInstance "/Game/ContentCreationTools/ShadowProxyGenerator/Materials/M_FrameCopy.M_FrameCopy" "None" "None" :!class /Script/Engine.KismetMaterialLibrary))
  (Variables|Default|SetConvertTo_DistanceField_MID _returnvalue_2) ; Converts a 2x2 atlas into RGBA subUV atlas
  (bind _returnvalue_3 (Rendering|Material|CreateDynamicMaterialInstance "/Game/ContentCreationTools/ShadowProxyGenerator/Materials/M_PreviewBlend.M_PreviewBlend" "None" "None" :!class /Script/Engine.KismetMaterialLibrary))
  (Variables|Default|SetPreviewBlendMID _returnvalue_3)
  (Utilities|Array|Clear (Variables|Debug|GetCaptureVectors))
  (for _index (range 4)
    (Utilities|Array|Add (Variables|Debug|GetCaptureVectors) (Math|Vector|RotateVectorAroundAxis "1.000000, 0, 0" (* (/ (Math|Conversions|ToFloat(Integer) _index) 3.0) -89.0) "0,1.000000,0.000000")))
  (Rendering|Material|SetMaterial (Variables|Default|Getcardtest) 0 _previewblendmid)
  (Rendering|Material|SetTextureParameterValue _previewblendmid "RT" (Variables|Default|GetFinalAtlas%20RT))
  (Rendering|Material|SetScalarParameterValue _previewblendmid "T" (Math|Float|Clamp(Float) (Variables|Default|GetT)) :!class /Script/Engine.MaterialInstanceDynamic))
